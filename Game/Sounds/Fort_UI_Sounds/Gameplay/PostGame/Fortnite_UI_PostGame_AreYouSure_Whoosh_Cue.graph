SoundCue /Game/Sounds/Fort_UI_Sounds/Gameplay/PostGame/Fortnite_UI_PostGame_AreYouSure_Whoosh_Cue
root=Modulator_2
nodes [
 {id=Modulator_2, children [WavePlayer_0], properties {volumeMax=1.5, volumeMin=1.3}}
 {id=WavePlayer_0, properties {soundWaveAssetPtr=/Game/Sounds/Fort_UI_Sounds/Gameplay/PostGame/wav/UI_FrontEnd_AreYouSure}}
]
properties {
 bOverrideConcurrency=true
 maxDistance=2097152
 soundClassObject=&Fort_UI
 volumeMultiplier=0.5
 concurrencyOverrides {retriggerTime=0.15}
}
